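# Post-sim Animation Blueprint for the DracoDueler tail accessory, adding trailing motion and situational pose offsets.
# A Trail controller drives the ten-bone chain, scaled by TrailAlpha and RelaxSpeed, and caches the result.
# Additive bone rotations bend that pose strongly while sliding or knee-sliding, slightly in the front end, blending between states.
# On customization load it caches pawn and body mesh, and sets "Hide Element 1" when a backbling or cape is worn.
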AnimBlueprint /BRCosmetics/Accessories/FORT_Tails/Tail_M_MED_DracoDueler/Meshes/M_MED_DracoDueler_Tail_AnimBP_PostSim
target_skeleton=/Game/Accessories/FORT_Tails/Common/Fortnite_Base_Tail_Skeleton
graphs [
 AnimGraph {nodes [
   n1 = "Transform (Modify) Bone - Bone: C_Tail_A_Base_Jnt"(ComponentPose=Local To Component(LocalPose=Use cached pose 'Pre_Pose'()), Alpha=1, Translation="0,0,0", Rotation="0,0,0", Scale="1,1,1") {
    rotationMode=BMM_Additive
    rotationSpace=BCS_ParentBoneSpace
    translationSpace=BCS_BoneSpace
    alphaBoolBlend {blendOption=Linear}
    alphaScaleBiasClamp {interpSpeedDecreasing=5}
    boneToModify {boneName=C_Tail_A_Base_Jnt}
   }
   n2 = "Transform (Modify) Bone - Bone: C_Tail_A_1_Jnt"(ComponentPose=@n1, Alpha=1, Translation="0,0,0", Rotation="5,0,0", Scale="1,1,1") {
    rotationMode=BMM_Additive
    rotationSpace=BCS_BoneSpace
    translationSpace=BCS_BoneSpace
    alphaBoolBlend {blendOption=Linear}
    alphaScaleBiasClamp {interpSpeedDecreasing=5}
    boneToModify {boneName=C_Tail_A_1_Jnt}
    rotation {pitch=5}
   }
   n3 = "Transform (Modify) Bone - Bone: C_Tail_A_2_Jnt"(ComponentPose=@n2, Alpha=1, Translation="0,0,0", Rotation="5,0,0", Scale="1,1,1") {
    rotationMode=BMM_Additive
    rotationSpace=BCS_BoneSpace
    translationSpace=BCS_BoneSpace
    alphaBoolBlend {blendOption=Linear}
    alphaScaleBiasClamp {interpSpeedDecreasing=5}
    boneToModify {boneName=C_Tail_A_2_Jnt}
    rotation {pitch=5}
   }
   n4 = "Transform (Modify) Bone - Bone: C_Tail_A_2_Jnt"(ComponentPose=@n3, Alpha=1, Translation="0,0,0", Rotation="3,0,0", Scale="1,1,1") {
    rotationMode=BMM_Additive
    rotationSpace=BCS_BoneSpace
    translationSpace=BCS_BoneSpace
    alphaBoolBlend {blendOption=Linear}
    alphaScaleBiasClamp {interpSpeedDecreasing=5}
    boneToModify {boneName=C_Tail_A_2_Jnt}
    rotation {pitch=3}
   }
   n5 = "Transform (Modify) Bone - Bone: C_Tail_A_3_Jnt"(ComponentPose=@n4, Alpha=1, Translation="0,0,0", Rotation="0,0,0", Scale="1,1,1") {
    rotationMode=BMM_Additive
    rotationSpace=BCS_BoneSpace
    translationSpace=BCS_BoneSpace
    alphaBoolBlend {blendOption=Linear}
    alphaScaleBiasClamp {interpSpeedDecreasing=5}
    boneToModify {boneName=C_Tail_A_3_Jnt}
   }
   n6 = Component To Local(ComponentPose=@n5)
   n7 = Use cached pose 'Pre_Pose'()
   n8 = Get bIsInFrontEnd()
   n9 = "Transform (Modify) Bone - Bone: C_Tail_A_Base_Jnt"(ComponentPose=Local To Component(LocalPose=Use cached pose 'Pre_Pose'()), Alpha=1, Translation="0,0,0", Rotation="0,0,75", Scale="1,1,1") {
    rotationMode=BMM_Additive
    rotationSpace=BCS_ParentBoneSpace
    alphaBoolBlend {blendOption=Linear}
    alphaScaleBiasClamp {interpSpeedDecreasing=5}
    boneToModify {boneName=C_Tail_A_Base_Jnt}
    rotation {roll=75}
   }
   n10 = "Transform (Modify) Bone - Bone: C_Tail_A_1_Jnt"(ComponentPose=@n9, Alpha=1, Translation="0,0,0", Rotation="45,0,0", Scale="1,1,1") {
    rotationMode=BMM_Additive
    rotationSpace=BCS_BoneSpace
    alphaBoolBlend {blendOption=Linear}
    alphaScaleBiasClamp {interpSpeedDecreasing=5}
    boneToModify {boneName=C_Tail_A_1_Jnt}
    rotation {pitch=45}
   }
   n11 = "Transform (Modify) Bone - Bone: C_Tail_A_2_Jnt"(ComponentPose=@n10, Alpha=1, Translation="0,0,0", Rotation="20,0,0", Scale="1,1,1") {
    rotationMode=BMM_Additive
    rotationSpace=BCS_BoneSpace
    alphaBoolBlend {blendOption=Linear}
    alphaScaleBiasClamp {interpSpeedDecreasing=5}
    boneToModify {boneName=C_Tail_A_2_Jnt}
    rotation {pitch=20}
   }
   n12 = "Transform (Modify) Bone - Bone: C_Tail_A_2_Jnt"(ComponentPose=@n11, Alpha=1, Translation="0,0,0", Rotation="10,0,0", Scale="1,1,1") {
    rotationMode=BMM_Additive
    rotationSpace=BCS_BoneSpace
    alphaBoolBlend {blendOption=Linear}
    alphaScaleBiasClamp {interpSpeedDecreasing=5}
    boneToModify {boneName=C_Tail_A_2_Jnt}
    rotation {pitch=10}
   }
   n13 = "Transform (Modify) Bone - Bone: C_Tail_A_3_Jnt"(ComponentPose=@n12, Alpha=1, Translation="0,0,0", Rotation="5,0,0", Scale="1,1,1") {
    rotationMode=BMM_Additive
    rotationSpace=BCS_BoneSpace
    alphaBoolBlend {blendOption=Linear}
    alphaScaleBiasClamp {interpSpeedDecreasing=5}
    boneToModify {boneName=C_Tail_A_3_Jnt}
    rotation {pitch=5}
   }
   Output Pose(Result=Blend Poses by bool(
     BlendPose_0=Component To Local(ComponentPose=@n13)
     BlendPose_1=Blend Poses by bool(BlendPose_0=@n6, BlendPose_1=@n7, BlendTime_0=0.25, BlendTime_1=0.25, bActiveValue=@n8) {blendType=Linear}
     BlendTime_0=0.25
     BlendTime_1=0.25
     bActiveValue=Get slide_var() -> (slide_var:bool)
    ) {blendType=Linear}) {layerGroup=None}
   n14 = Trail controller - Bone: C_Tail_A_9_Jnt!(
    ComponentPose=Local To Component(LocalPose=Input Pose())
    Alpha=Get TrailAlpha() -> (TrailAlpha:double)
    RelaxationSpeedScale:float=Get RelaxSpeed() -> (RelaxSpeed:double)
   ) -> (Pose) {
    bLimitRotation=true
    chainLength=10
    maxDeltaTime=0.0333
    alphaScaleBiasClamp {bInterpResult=true}
    rotationLimits [
     {limitMax=<180 180 180>}
     {limitMax=<180 180 180>}
     {limitMax=<180 180 180>}
     {limitMax=<180 180 180>}
     {limitMax=<180 180 180>}
     {limitMax=<180 180 180>}
     {limitMax=<180 180 180>}
     {limitMax=<180 180 180>}
     {limitMax=<180 180 180>}
     {limitMax=<180 180 180>}
    ]
    rotationOffsets [<0 0 0>, <0 0 0>, <0 0 0>, <0 0 0>, <0 0 0>, <0 0 0>, <0 0 0>, <0 0 0>, <0 0 0>, <0 0 0>]
    trailBone {boneName=C_Tail_A_9_Jnt}
    trailRelaxationSpeed {editorCurveData {keys [<at 0 10 RCIM_Linear>,<at 0.5107 8.1304 RCIM_Linear>,<at 1 5 RCIM_Linear>]}}
   }
   Save cached pose 'Pre_Pose'(Pose=Component To Local(ComponentPose=@n14)) {cacheName=Pre_Pose}
  ]}
 EventGraph {nodes [
   Event BlueprintUpdateAnimation()
   > Set slide_var(slide_var:bool=OR Boolean(A=Get bIsSliding(), B=Get bIsKneeSliding())) -> (Output_Get:bool)
   Event OnCharacterCustomizationLoaded()
   > n1 = Sequence()
   n2 = CheckForBackbling()
   > Set isWearingBackbling(isWearingBackbling=@n2)
   > Set isWearingCape(isWearingCape=@n2.bIsBackblingLong)
   > Branch(Condition=OR Boolean(A=Get isWearingCape() -> (isWearingCape:bool), B=Get isWearingBackbling()))
   > n3 = Get Dynamic Material Instance(SkeletalMeshComponent=GetPartSkeletalMeshComponent())
   > SetScalarParameterValue(self=@n3, ParameterName=Hide Element 1, Value=1)
   n4 = Cast To PlayerPawn_Athena(execute=@n1.then_1, Object=TryGetPawnOwner())
   > n5 = Set PawnOwner(PawnOwner:Player Pawn Athena Object Reference=@n4) -> (Output_Get:Player Pawn Athena Object Reference)
   > Set BodyMesh(BodyMesh=GetSkeletalMeshForPartType(self=@n5))
  ]}
]
variables [
 BodyMesh {type=SkeletalMeshComponent}
 isWearingBackbling {type=boolean}
 isWearingCape {type=boolean}
 PawnOwner {type=PlayerPawn_Athena_C}
 RelaxSpeed {type=number}
 slide_var {type=boolean}
 TrailAlpha {type=number}
]
